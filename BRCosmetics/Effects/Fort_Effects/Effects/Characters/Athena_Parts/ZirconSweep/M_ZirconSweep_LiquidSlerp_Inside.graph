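# Two-sided unlit masked material for the inside of the ZirconSweep character effect's liquid, usable on Niagara mesh particles.
# Panning, distorted cloud noise with BumpOffset parallax blends two blue colors under vertical UV gradients, with time-of-day brightening.
# A faint fake refraction sampled from the skylight environment map is added, then multiplied by a fresnel-masked color for emissive.
# A separate "Outside" noise branch feeds nothing in the shown graph.
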Material /BRCosmetics/Effects/Fort_Effects/Effects/Characters/Athena_Parts/ZirconSweep/M_ZirconSweep_LiquidSlerp_Inside
graph {
 properties {
  bUsedWithNiagaraMeshParticles=true
  blendMode=BLEND_Masked
  pixelDepthOffsetMode=PDOM_Legacy
  shadingModel=MSM_Unlit
  twoSided=true
  usedShadingModels=Unlit
 }
 nodes [
  Constant()
  Constant() {r=0.95}
  Constant() {r=0.025}
  n1 = Multiply(A=TextureCoordinate(), B=AppendVector(A=Constant() {r=0.2}, B=Constant() {r=1}))
  n2 = AppendVector(A=ScalarParameter() {defaultValue=-0.2, parameterName=Outside_Pan_X}, B=ScalarParameter() {defaultValue=0.1, parameterName=Outside_Pan_Y})
  n3 = Constant() {r=0.2}
  n4 = Constant() {r=1}
  n5 = Panner(
   Coordinate=Multiply(A=TextureCoordinate(), B=AppendVector(A=@n3, B=@n4))
   Speed=AppendVector(A=ScalarParameter() {defaultValue=-0.2, parameterName=Outside_Pan_X}, B=ScalarParameter() {defaultValue=0.1, parameterName=Outside_Pan_Y})
  )
  n6 = Multiply(
   A=Power(
    Base=TextureSample(UVs=@n5) {texture=&AtmosphericCloudNoise01}.R
    Exp=ScalarParameter() {defaultValue=1, parameterName=Outside_Noise_Distortion_Power}
   )
   B=ScalarParameter() {defaultValue=0.2, parameterName=Outside_Noise_Distortion_Mult}
  )
  n7 = TextureSample(UVs=Add(A=Panner(Coordinate=@n1, Speed=@n2), B=Clamp(None=@n6))) {texture=&AtmosphericCloudNoise01}
  n8 = ScalarParameter() {defaultValue=2, parameterName=Outside_Texture_Power}
  n9 = Clamp(None=Multiply(A=Power(Base=@n7.RGB, Exp=@n8), B=ScalarParameter() {defaultValue=3, parameterName=Outside_Texture_Mult}))
  LinearInterpolate(A=Constant3Vector() {constant=<rgba 0.013 0.3864 0.8469 1>}, B=Constant3Vector() {constant=<rgba 0.0024 0.3325 0.7605 1>}, Alpha=@n9)
  n10 = CameraVectorWS()
  n11 = Normalize(VectorInput=PixelNormalWS())
  n12 = Subtract(A=DotProduct(A=@n10, B=@n11)) {constB=0.15}
  n13 = LinearInterpolate(Alpha=Power(Base=Saturate(None=@n12), Exp=Constant() {r=0.6366}) {constExponent=0.5}) {constA=1, constB=0}
  n14 = SkyLightEnvMapSample(
   Direction=Subtract(A=Multiply(A=@n13, B=@n11), B=@n10) {constA=0, constB=0}
   Roughness=ScalarParameter() {parameterName=Refraction Roughness, sliderMax=1}
  )
  n15 = ScalarParameter() {parameterName=Refraction Min Value, sliderMax=1}
  n16 = ScalarParameter() {defaultValue=1, parameterName=Refraction Saturation, sliderMax=1, sliderMin=-1}
  n17 = Multiply(
   A=Desaturation(None=Max(A=@n14, B=@n15), Fraction=OneMinus(None=@n16))
   B=ScalarParameter() {defaultValue=0.025, parameterName=Fakereflection_Intensity}
  )
  n18 = TextureCoordinate()
  n19 = Power(Base=ComponentMask(None=@n18) {g=true}, Exp=ScalarParameter() {defaultValue=1, parameterName=Param_1})
  n20 = ScalarParameter() {defaultValue=1, parameterName=Param_3}
  MaterialExpressionRemap_1 = MaterialXRemap!(
   None=Clamp(None=Multiply(A=@n19, B=@n20))
   InputLow
   InputHigh
   TargetLow=ScalarParameter() {defaultValue=1, parameterName=TargetLow}
   TargetHigh=ScalarParameter() {defaultValue=0.1, parameterName=TargetHigh}
  ) {targetHighDefault=0.2, targetLowDefault=0.2}
  n21 = ScalarParameter() {defaultValue=0.25, parameterName=Noise_Tile_X}
  n22 = ScalarParameter() {defaultValue=1, parameterName=Noise_Tile_Y}
  n23 = Panner(
   Coordinate=Multiply(A=TextureCoordinate(), B=AppendVector(A=@n21, B=@n22))
   Speed=AppendVector(A=ScalarParameter() {defaultValue=-0.025, parameterName=Pan_X}, B=ScalarParameter() {defaultValue=0.0125, parameterName=Pan_Y})
  )
  n24 = TextureCoordinate()
  n25 = AppendVector(A=ScalarParameter() {defaultValue=0.5, parameterName=Distort_Tile_X}, B=ScalarParameter() {defaultValue=1, parameterName=Distort_Tile_Y})
  n26 = Multiply(A=ScalarParameter() {defaultValue=-0.025, parameterName=Pan_X}) {constB=-1}
  n27 = ScalarParameter() {defaultValue=0.0125, parameterName=Pan_Y}
  n28 = TextureSample(UVs=Panner(Coordinate=Multiply(A=@n24, B=@n25), Speed=AppendVector(A=@n26, B=@n27))) {texture=&AtmosphericCloudNoise01}
  n29 = ScalarParameter() {defaultValue=1, parameterName=Noise_Distortion_Power}
  n30 = Clamp(None=Multiply(A=Power(Base=@n28.R, Exp=@n29), B=ScalarParameter() {defaultValue=0.2, parameterName=Noise_Distortion_Mult}))
  n31 = Multiply(A=TextureCoordinate(), B=AppendVector(A=Constant() {r=0.5}, B=Constant() {r=0.5}))
  n32 = AppendVector(A=Constant() {r=-0.025}, B=Constant() {r=0.025})
  MaterialExpressionRemap_0 = MaterialXRemap(None=TextureSample(UVs=Panner(Coordinate=@n31, Speed=@n32)) {texture=&AtmosphericCloudNoise01}.RGB) {targetLowDefault=0.5}
  n33 = ScalarParameter() {defaultValue=7, parameterName=BumpOffset-Intensity}
  n34 = TextureSample(UVs=BumpOffset(Coordinate=Add(A=@n23, B=@n30), Height=Multiply(A=@MaterialExpressionRemap_0, B=@n33) {constB=-2})) {texture=&AtmosphericCloudNoise01}
  n35 = ScalarParameter() {defaultValue=1, parameterName=Texture_Power}
  n36 = Clamp(None=Multiply(A=Power(Base=@n34.R, Exp=@n35), B=ScalarParameter() {defaultValue=1, parameterName=Texture_Mult}))
  n37 = ComponentMask(None=OneMinus(None=TextureCoordinate())) {g=true}
  n38 = ScalarParameter() {defaultValue=1, parameterName=Fade_Power}
  n39 = Clamp(None=Multiply(A=Power(Base=@n37, Exp=@n38), B=ScalarParameter() {defaultValue=5, parameterName=Fade_Mult}))
  n40 = Clamp(None=LinearInterpolate(
    A=VectorParameter() {parameterName=Inside_Color_A, defaultValue=<rgba 0 0.382 1 1>}.RGB
    B=VectorParameter() {parameterName=Inside_Color_B, defaultValue=<rgba 0.0019 0.2623 0.6 1>}.RGB
    Alpha=Multiply(A=@n36, B=@n39)
   ))
  n41 = ScalarParameter() {defaultValue=6, parameterName=ColorMult_Day}
  n42 = MaterialFunctionCall(Color_Morning=@n41, Color_Day=@n41, Color_Evening=@n41, Color_Night=ScalarParameter() {defaultValue=3, parameterName=ColorMult_Night}) {materialFunction=&Color_Time_of_Day}
  n43 = OneMinus(None=TextureCoordinate())
  n44 = Multiply(
   A=Power(Base=ComponentMask(None=@n43) {g=true}, Exp=ScalarParameter() {defaultValue=1, parameterName=Param_2})
   B=ScalarParameter() {defaultValue=1, parameterName=Param_4}
  )
  n45 = Multiply(A=@MaterialExpressionRemap_1, B=LinearInterpolate(A=@n40, B=Multiply(A=@n42.Result, B=@n40), Alpha=Clamp(None=@n44)))
  n46 = VectorParameter() {parameterName=Fresnel_Color, defaultValue=<rgba 1 1 1 1>}
  n47 = Clamp(None=MaterialFunctionCall(
    Brightness=ScalarParameter() {defaultValue=1.75, parameterName=Fresnel_Brightness}
    Exponent=ScalarParameter() {defaultValue=5, parameterName=Fresnel_Exponent}
   ) {materialFunction=&MF_Fresnel_FX}.Fresnel Double Sided)
  MaterialFunctionCall(Emissive=Multiply(A=Add(A=@n17, B=@n45), B=LinearInterpolate(B=@n46.RGB, Alpha=@n47))) {materialFunction=&MF_FX_EmissiveAdjustments}
  n48 = Power(Base=ComponentMask(None=TextureCoordinate()) {g=true}, Exp=ScalarParameter() {defaultValue=4, parameterName=Outside_Fade_Power})
  n49 = ScalarParameter() {defaultValue=2, parameterName=Outside_Fade_Mult}
  Multiply(A=@n9, B=Clamp(None=Multiply(A=@n48, B=@n49)))
  Multiply(A=TwoSidedSign()) {constB=-1}
  OneMinus(None=@n18)
  ScalarParameter() {defaultValue=2}
  Subtract(A=MaterialFunctionCall(
    Brightness=ScalarParameter() {defaultValue=4, parameterName=Fresnel_Mult}
    Exponent=ScalarParameter() {defaultValue=1, parameterName=Fresnel_Power}
   ) {materialFunction=&MF_Fresnel_FX}.Fresnel Double Sided)
  TransformPosition() {transformType=TRANSFORMPOSSOURCE_World}
 ]
}
